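# Physics asset for the torso mesh of the female medium F_MED_Lateral player body, covering secondary-motion bones for hood, necklace, strings and vest flap.
# Capsules on spine, upper arms and right clavicle form the body volume, linked by free angular constraints.
# Two three-link string chains, a two-link necklace, a hood and a right vest flap hang off spine_05 with tight swing limits and zero-degree twist limits.
# There is no left clavicle body, so upperarm_l constrains directly to spine_05.
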PhysicsAsset /BRCosmetics/Characters/Player/Female/Medium/Bodies/F_MED_Lateral/Meshes/F_MED_Lateral_Torso_Physics
bodies {
 clavicle_r [{
   shapeType=Capsule
   length=10.3165
   radius=7.3306
   shapeName=clavicle_r_capsule
   center=<-5.6114 -0.0001 3.1882>
   rotation=<rot 90 -1.2658 -1.2657>
  }]
 dyn_hood_c_01 [{shapeType=Capsule, length=12.5248, radius=2.5797, shapeName=dyn_hood_c_01_capsule, center=<0.679 -2.3 0>, rotation=<rot 0 -90 0>}]
 dyn_necklace_c_01 [{
   shapeType=Capsule
   length=3.25
   radius=1.25
   shapeName=dyn_necklace_c_01_capsule
   center=<1.3988 0.8744 -0.0001>
   rotation=<rot 90 0 -11.4037>
  }]
 dyn_necklace_c_02 [{shapeType=Capsule, length=3.25, radius=0.75, shapeName=dyn_necklace_c_02_capsule, center=<2.1073 0 0>, rotation=<rot 90 0 0>}]
 dyn_string_l_01 [{shapeType=Capsule, length=4.75, radius=0.75, shapeName=dyn_string_l_01_capsule, center=<3.7706 0.5823 0>, rotation=<rot 90 -90 -84.1685>}]
 dyn_string_l_02 [{shapeType=Capsule, length=5, radius=1, shapeName=dyn_string_l_02_capsule, center=<3.1949 0.0001 0>, rotation=<rot 90 0 0>}]
 dyn_string_l_03 [{shapeType=Capsule, length=5, radius=1, shapeName=dyn_string_l_03_capsule, center=<3.7987 0 0>, rotation=<rot 83.3656 179.9999 179.9999>}]
 dyn_string_r_01 [{shapeType=Capsule, length=6.5, radius=1, shapeName=dyn_string_r_01_capsule, center=<3.7942 0.2826 0>, rotation=<rot 90 0.8952 3.0249>}]
 dyn_string_r_02 [{shapeType=Capsule, length=4, radius=1, shapeName=dyn_string_r_02_capsule, center=<3.1949 0.0001 0>, rotation=<rot 90 0 0>}]
 dyn_string_r_03 [{shapeType=Capsule, length=4, radius=0.75, shapeName=dyn_string_r_03_capsule, center=<3.2667 0.0001 -0.3872>, rotation=<rot 85.7301 0 0>}]
 dyn_vestFlap_r_01 [{
   shapeType=Capsule
   length=2
   radius=0.75
   shapeName=dyn_vestFlap_r_01_capsule
   center=<1.6107 0.6879 -0.0001>
   rotation=<rot 90 -0.0099 -0.0099>
  }]
 spine_03 [{shapeType=Capsule, length=5.75, radius=11.5, shapeName=spine_03_capsule, center=<5.9813 -0.6497 0>, rotation=<rot 0 96.8169 0>}]
 spine_04 [{shapeType=Capsule, length=7.6183, radius=10.702, shapeName=spine_04_capsule, center=<2.188 -0.4695 0>, rotation=<rot 0 89.9999 0>}]
 spine_05 [
  {shapeType=Capsule, length=17.5, radius=9.5, shapeName=spine_05_capsule, center=<-0.0005 -1.0883 0>}
  {shapeType=Capsule, length=22, radius=8, shapeName=spine_05_capsule, center=<3.727 -1.5309 0.0001>}
  {shapeType=Capsule, length=16.25, radius=8.5, shapeName=spine_05_capsule, center=<3.8711 -1.5878 0.2255>, rotation=<rot 0 0 -6.5925>}
 ]
 upperarm_l [{shapeType=Capsule, length=15.9117, radius=5.8555, shapeName=upperarm_l_capsule, center=<11.0519 -0.0002 0>, rotation=<rot -90 0 0>}]
 upperarm_r [{shapeType=Capsule, length=19.6835, radius=4.4055, shapeName=upperarm_r_capsule, center=<-11.0519 -0.0002 0>, rotation=<rot 90 0 0>}]
}
constraints [
 {
  bone1Name=spine_05
  bone2Name=spine_04
  swing1LimitDegrees=45
  swing1Motion=Free
  swing2LimitDegrees=45
  swing2Motion=Free
  twistLimitDegrees=45
  twistMotion=Free
 }
 {bone1Name=dyn_string_l_01, bone2Name=spine_05, swing1LimitDegrees=3, swing2LimitDegrees=3, twistLimitDegrees=0}
 {bone1Name=dyn_string_l_02, bone2Name=dyn_string_l_01, swing1LimitDegrees=10, swing2LimitDegrees=10, twistLimitDegrees=0}
 {bone1Name=dyn_string_l_03, bone2Name=dyn_string_l_02, swing1LimitDegrees=10, swing2LimitDegrees=10, twistLimitDegrees=0}
 {bone1Name=dyn_string_r_01, bone2Name=spine_05, swing1LimitDegrees=3, swing2LimitDegrees=3, twistLimitDegrees=0}
 {bone1Name=dyn_string_r_02, bone2Name=dyn_string_r_01, swing1LimitDegrees=10, swing2LimitDegrees=10, twistLimitDegrees=0}
 {bone1Name=dyn_string_r_03, bone2Name=dyn_string_r_02, swing1LimitDegrees=10, swing2LimitDegrees=10, twistLimitDegrees=0}
 {bone1Name=dyn_vestFlap_r_01, bone2Name=spine_05, swing1LimitDegrees=35, swing2LimitDegrees=0, twistLimitDegrees=0}
 {bone1Name=dyn_necklace_c_01, bone2Name=spine_05, swing1LimitDegrees=10, swing2LimitDegrees=5, twistLimitDegrees=0}
 {bone1Name=dyn_necklace_c_02, bone2Name=dyn_necklace_c_01, swing1LimitDegrees=10, swing2LimitDegrees=10, twistLimitDegrees=0}
 {
  bone1Name=spine_04
  bone2Name=spine_03
  swing1LimitDegrees=45
  swing1Motion=Free
  swing2LimitDegrees=45
  swing2Motion=Free
  twistLimitDegrees=45
  twistMotion=Free
 }
 {bone1Name=dyn_hood_c_01, bone2Name=spine_05, swing1LimitDegrees=20, swing2LimitDegrees=20, twistLimitDegrees=0}
 {
  bone1Name=upperarm_l
  bone2Name=spine_05
  swing1LimitDegrees=45
  swing1Motion=Free
  swing2LimitDegrees=45
  swing2Motion=Free
  twistLimitDegrees=45
  twistMotion=Free
 }
 {
  bone1Name=upperarm_r
  bone2Name=clavicle_r
  swing1LimitDegrees=45
  swing1Motion=Free
  swing2LimitDegrees=45
  swing2Motion=Free
  twistLimitDegrees=45
  twistMotion=Free
 }
 {
  bone1Name=clavicle_r
  bone2Name=spine_05
  swing1LimitDegrees=45
  swing1Motion=Free
  swing2LimitDegrees=45
  swing2Motion=Free
  twistLimitDegrees=45
  twistMotion=Free
 }
]
