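# Physics asset for the crouch variant of the F_MED_Lateral female player body, setting up its dynamic vest bones and lower-body collision.
# Capsules on pelvis, spine_01 and both thighs form the body volume, with extra offset capsules on pelvis and right thigh.
# Three back and two front vest bones hang from the pelvis on swing-limited, zero-twist constraints, with each front side chaining to a second segment.
# Spine and thigh constraints to the pelvis leave rotation free.
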PhysicsAsset /BRCosmetics/Characters/Player/Female/Medium/Bodies/F_MED_Lateral/Meshes/F_MED_Lateral_crouch_Physics
bodies {
 dyn_vestBack_c_01 [{shapeType=Capsule, length=7.75, radius=2.25, shapeName=dyn_vestBack_c_01_capsule, center=<5.3475 -0.6801 0>, rotation=<rot 90 0 0>}]
 dyn_vestBack_l_01 [{shapeType=Capsule, length=7.5, radius=2.5, shapeName=dyn_vestBack_l_01_capsule, center=<5.5965 -1.1308 0.1633>, rotation=<rot 90 0 0>}]
 dyn_vestBack_r_01 [{shapeType=Capsule, length=8, radius=2, shapeName=dyn_vestBack_l_01_capsule, center=<5.5379 2.2284 0.2084>, rotation=<rot 90 0 0>}]
 dyn_vestFront_l_01 [{
   shapeType=Capsule
   length=12.25
   radius=1.75
   shapeName=dyn_vestFront_l_01_capsule
   center=<8.1777 0.3979 -0.5615>
   rotation=<rot 90 -1.1189 0.2316>
  }]
 dyn_vestFront_l_02 [{
   shapeType=Capsule
   length=9.25
   radius=1.5
   shapeName=dyn_vestFront_l_02_capsule
   center=<5.6 0.2 3>
   rotation=<rot -0.621 -179.9999 -174.9567>
  }]
 dyn_vestFront_r_01 [{shapeType=Capsule, length=14, radius=1.75, shapeName=dyn_vestFront_r_01_capsule, center=<8.91 1.3791 0>, rotation=<rot 90 90 -94.7381>}]
 dyn_vestFront_r_02 [{shapeType=Capsule, length=1.75, radius=1, shapeName=dyn_vestFront_r_02_capsule, center=<2.0021 0 0>, rotation=<rot 90 0 0>}]
 pelvis [
  {shapeType=Capsule, length=11.75, radius=9, shapeName=pelvis_capsule, center=<-0.0001 -0.9054 0>}
  {shapeType=Capsule, length=10.75, radius=9.25, shapeName=pelvis_capsule, center=<-4.0734 -2.8672 -0.0001>}
  {
   shapeType=Capsule
   length=15.75
   radius=7.5
   shapeName=pelvis_capsule
   center=<-9.5336 14.0009 3.1867>
   rotation=<rot -3.1956 94.8775 77.2732>
  }
 ]
 spine_01 [{shapeType=Capsule, length=9, radius=10.75, shapeName=spine_01_capsule, center=<0 0.5003 0>}]
 thigh_l [{
   shapeType=Capsule
   length=31.3661
   radius=9.8802
   shapeName=thigh_l_capsule
   center=<-14.9759 0.2125 2.1674>
   rotation=<rot 86.5201 -180 -180>
  }]
 thigh_r [
  {shapeType=Capsule, length=41.0421, radius=8.4966, shapeName=thigh_r_capsule, center=<12.6854 0.1308 -1.6722>, rotation=<rot 90 0 0>}
  {shapeType=Capsule, length=32, radius=7.75, shapeName=thigh_r_capsule, center=<10.1234 -10.3651 -12.7298>, rotation=<rot 0 0 37.1006>}
 ]
}
constraints [
 {
  bone1Name=spine_01
  bone2Name=pelvis
  swing1LimitDegrees=45
  swing1Motion=Free
  swing2LimitDegrees=45
  swing2Motion=Free
  twistLimitDegrees=45
  twistMotion=Free
 }
 {bone1Name=dyn_vestBack_c_01, bone2Name=pelvis, swing1LimitDegrees=20, swing2LimitDegrees=10, twistLimitDegrees=0}
 {bone1Name=dyn_vestBack_l_01, bone2Name=pelvis, swing1LimitDegrees=20, swing2LimitDegrees=10, twistLimitDegrees=0}
 {bone1Name=dyn_vestBack_r_01, bone2Name=pelvis, swing1LimitDegrees=20, swing2LimitDegrees=2, twistLimitDegrees=0}
 {bone1Name=dyn_vestFront_l_01, bone2Name=pelvis, swing1LimitDegrees=40, swing2LimitDegrees=35, twistLimitDegrees=0}
 {bone1Name=dyn_vestFront_r_01, bone2Name=pelvis, swing1LimitDegrees=30, swing2LimitDegrees=35, twistLimitDegrees=0}
 {bone1Name=dyn_vestFront_r_02, bone2Name=dyn_vestFront_r_01, swing1LimitDegrees=30, swing2LimitDegrees=30, twistLimitDegrees=0}
 {
  bone1Name=thigh_l
  bone2Name=pelvis
  swing1LimitDegrees=45
  swing1Motion=Free
  swing2LimitDegrees=45
  swing2Motion=Free
  twistLimitDegrees=45
  twistMotion=Free
 }
 {
  bone1Name=thigh_r
  bone2Name=pelvis
  swing1LimitDegrees=45
  swing1Motion=Free
  swing2LimitDegrees=45
  swing2Motion=Free
  twistLimitDegrees=45
  twistMotion=Free
 }
 {bone1Name=dyn_vestFront_l_02, bone2Name=dyn_vestFront_l_01, swing1LimitDegrees=30, swing2LimitDegrees=30, twistLimitDegrees=0}
]
