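# Animation Blueprint for the F_MED_Lateral female cosmetic body, adding secondary motion to vest, hood and patches.
# It copies the parent mesh pose through twist and sculpt Control Rigs, caches it, blends standing and crouch RigidBody sims by crouch state, then runs a torso sim.
# Bone constraints and additive transforms then correct dyn bones, weighted by update-computed alphas from thigh, arm and head rotation, jumping, skydiving and front-end state.
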AnimBlueprint /BRCosmetics/Characters/Player/Female/Medium/Bodies/F_MED_Lateral/Meshes/F_MED_Lateral_AnimBP
target_skeleton=&/F_MED_Lateral_Skeleton
graphs [
 AnimGraph {nodes [
   n1 = RigidBody!(
    ComponentPose=Local To Component(LocalPose=Use cached pose 'Anim Cache'())
    Alpha
    OverrideWorldGravity:Vector=Get Gravity Override()
    ExternalForce
   ) -> (Pose) {
    bClampLinearTranslationLimitToRefPose=true
    bOverrideWorldGravity=true
    overridePhysicsAsset=&/F_MED_Lateral_skelMesh_Physics
    simulationSpace=BaseBoneSpace
    alphaBoolBlend {blendOption=Linear}
    baseBoneRef {boneName=root}
    componentAppliedLinearAccClamp=<6000 6000 6000>
    componentLinearAccScale=<0.01 0.01 0.01>
    componentLinearVelScale=<0.01 0.01 0.01>
    overrideWorldGravity {z=-980}
   }
   n2 = Component To Local(ComponentPose=@n1)
   n3 = RigidBody(ComponentPose=Local To Component(LocalPose=Use cached pose 'Anim Cache'()), OverrideWorldGravity=Get Gravity Override()) {
    bClampLinearTranslationLimitToRefPose=true
    bOverrideWorldGravity=true
    overridePhysicsAsset=&/F_MED_Lateral_crouch_Physics
    simulationSpace=BaseBoneSpace
    alphaBoolBlend {blendOption=Linear}
    baseBoneRef {boneName=root}
    componentAppliedLinearAccClamp=<6000 6000 6000>
    componentLinearAccScale=<0.01 0.01 0.01>
    componentLinearVelScale=<0.01 0.01 0.01>
    overrideWorldGravity {z=-980}
   }
   n4 = Component To Local(ComponentPose=@n3)
   n5 = Get Crouching() -> (Crouching:bool)
   n6 = RigidBody(
    ComponentPose=Local To Component(LocalPose=Blend Poses by bool(BlendPose_0=@n4, BlendPose_1=@n2, bActiveValue=@n5) {blendType=Linear})
    OverrideWorldGravity="0,0,-980"
   ) {
    bClampLinearTranslationLimitToRefPose=true
    bOverrideWorldGravity=true
    overridePhysicsAsset=&/F_MED_Lateral_Torso_Physics
    simulationSpace=BaseBoneSpace
    alphaBoolBlend {blendOption=Linear}
    baseBoneRef {boneName=spine_04}
    componentAppliedLinearAccClamp=<3000 3000 3000>
    componentLinearAccScale {x=1}
    componentLinearVelScale {x=1}
    overrideWorldGravity {z=-980}
    simSpaceSettings {
     maxAngularAcceleration=500
     maxAngularVelocity=3
     maxLinearAcceleration=500
     maxLinearVelocity=500
     velocityScaleZ=0
     worldAlpha=1
     externalLinearDragV=<0.2 0.2>
    }
   }
   n7 = "Transform (Modify) Bone - Bone: dyn_vestFront_l_01"(ComponentPose=@n6, Alpha=Get thigh L Rot() -> (thigh L Rot:double), Translation="2,0,0", Rotation="5,0,0", Scale="0.8,0.8,0.8") {
    rotationMode=BMM_Additive
    rotationSpace=BCS_BoneSpace
    scaleMode=BMM_Replace
    translationMode=BMM_Additive
    translationSpace=BCS_BoneSpace
    alphaBoolBlend {blendOption=Linear}
    boneToModify {boneName=dyn_vestFront_l_01}
    rotation {pitch=5}
    scale=<0.8 0.8 0.8>
    translation {x=2}
   }
   n8 = "Constraint - dyn_patch_2 (PPP)"!(ComponentPose=@n7, Alpha=1, ConstraintWeights_0=0.5, ConstraintWeights_1:float=0.1, ConstraintWeights_2:float=0.4) -> (Pose) {
    alphaBoolBlend {blendOption=Linear}
    boneToModify {boneName=dyn_patch_2}
    constraintSetup [{targetBone {boneName=spine_05}}, {targetBone {boneName=spine_04}}, {targetBone {boneName=clavicle_l}}]
    constraintWeights [1,1,1]
   }
   n9 = "Constraint - dyn_patch_1 (PPP)"!(ComponentPose=@n8, Alpha=1, ConstraintWeights_0=0.3, ConstraintWeights_1:float=0.4, ConstraintWeights_2:float=0.3) -> (Pose) {
    alphaBoolBlend {blendOption=Linear}
    boneToModify {boneName=dyn_patch_1}
    constraintSetup [{targetBone {boneName=spine_05}}, {targetBone {boneName=spine_04}}, {targetBone {boneName=clavicle_l}}]
    constraintWeights [1,1,1]
   }
   n10 = Get arm L Rot() -> (arm L Rot:double)
   n11 = "Transform (Modify) Bone - Bone: dyn_patch_1"(ComponentPose=@n9, Alpha=@n10, Translation="-0.5,1,0", Rotation="-10,0,0", Scale="1,1,1") {
    rotationMode=BMM_Additive
    rotationSpace=BCS_BoneSpace
    translationMode=BMM_Additive
    translationSpace=BCS_BoneSpace
    alphaBoolBlend {blendOption=Linear}
    boneToModify {boneName=dyn_patch_1}
    rotation {pitch=-10}
    translation=<-0.5 1>
   }
   n12 = "Transform (Modify) Bone - Bone: dyn_patch_2"(ComponentPose=@n11, Alpha=@n10, Translation="0.5,2,0", Rotation="-10,8,0", Scale="1,1,1") {
    rotationMode=BMM_Additive
    rotationSpace=BCS_BoneSpace
    translationMode=BMM_Additive
    translationSpace=BCS_BoneSpace
    alphaBoolBlend {blendOption=Linear}
    boneToModify {boneName=dyn_patch_2}
    rotation {pitch=-10, yaw=8}
    translation=<0.5 2>
   }
   n13 = Get isFrontEndAlpha() -> (isFrontEndAlpha:double)
   n14 = "Constraint - dyn_patch_1 (PP)"!(ComponentPose=@n12, Alpha=@n13, ConstraintWeights_0=0.5, ConstraintWeights_1:float=0.5) -> (Pose) {
    alphaBoolBlend {blendOption=Linear}
    boneToModify {boneName=dyn_patch_1}
    constraintSetup [{targetBone {boneName=spine_04}}, {targetBone {boneName=spine_05}}]
    constraintWeights [1,1]
   }
   n15 = "Constraint - dyn_patch_2 (PP)"!(ComponentPose=@n14, Alpha=@n13, ConstraintWeights_0=0.5, ConstraintWeights_1:float=0.5) -> (Pose) {
    alphaBoolBlend {blendOption=Linear}
    boneToModify {boneName=dyn_patch_2}
    constraintSetup [{targetBone {boneName=spine_04}}, {targetBone {boneName=spine_05}}]
    constraintWeights [1,1]
   }
   n16 = "Transform (Modify) Bone - Bone: dyn_hood_c_01"(ComponentPose=@n15, Alpha=Get start jump() -> (start jump:double), Translation="-1,-1.5,0", Rotation="0,5,0", Scale="1,1,1") {
    rotationMode=BMM_Additive
    rotationSpace=BCS_BoneSpace
    translationMode=BMM_Additive
    translationSpace=BCS_BoneSpace
    alphaBoolBlend {blendOption=Linear}
    alphaScaleBiasClamp {bInterpResult=true, interpSpeedDecreasing=1}
    boneToModify {boneName=dyn_hood_c_01}
    rotation {yaw=5}
    translation=<-1 -1.5>
   }
   n17 = "Transform (Modify) Bone - Bone: dyn_hood_c_01"(ComponentPose=@n16, Alpha=Get head Rot Back() -> (head Rot Back:double), Translation="3,-2,-5", Rotation="5,0,0", Scale="1,1,1") {
    rotationMode=BMM_Additive
    rotationSpace=BCS_BoneSpace
    translationMode=BMM_Additive
    translationSpace=BCS_BoneSpace
    alphaBoolBlend {blendOption=Linear}
    alphaScaleBiasClamp {bMapRange=true, interpSpeedDecreasing=1}
    boneToModify {boneName=dyn_hood_c_01}
    rotation {pitch=5}
    translation=<3 -2 -5>
   }
   n18 = Get Is in Air() -> (Is in Air:double)
   n19 = "Transform (Modify) Bone - Bone: dyn_hood_c_01"(ComponentPose=@n17, Alpha=@n18, Translation="-3,-2,0", Rotation="0,-5,0", Scale="1,1,1") {
    rotationMode=BMM_Additive
    rotationSpace=BCS_BoneSpace
    translationMode=BMM_Additive
    translationSpace=BCS_BoneSpace
    alphaBoolBlend {blendOption=Linear}
    alphaScaleBiasClamp {bInterpResult=true, interpSpeedDecreasing=1}
    boneToModify {boneName=dyn_hood_c_01}
    rotation {yaw=-5}
    translation=<-3 -2>
   }
   n20 = "Transform (Modify) Bone - Bone: dyn_vestBack_c_01"(ComponentPose=@n19, Alpha=@n18, Translation="0,-5,0", Rotation="0,-10,0", Scale="1,1,1") {
    rotationMode=BMM_Additive
    rotationSpace=BCS_BoneSpace
    translationMode=BMM_Additive
    translationSpace=BCS_BoneSpace
    alphaBoolBlend {blendOption=Linear}
    alphaScaleBiasClamp {bInterpResult=true, interpSpeedDecreasing=1}
    boneToModify {boneName=dyn_vestBack_c_01}
    rotation {yaw=-10}
    translation {y=-5}
   }
   n21 = "Transform (Modify) Bone - Bone: dyn_vestBack_l_01"(ComponentPose=@n20, Alpha=@n18, Translation="0,-5,0", Rotation="0,-10,0", Scale="1,1,1") {
    rotationMode=BMM_Additive
    rotationSpace=BCS_BoneSpace
    translationMode=BMM_Additive
    translationSpace=BCS_BoneSpace
    alphaBoolBlend {blendOption=Linear}
    alphaScaleBiasClamp {bInterpResult=true, interpSpeedDecreasing=1}
    boneToModify {boneName=dyn_vestBack_l_01}
    rotation {yaw=-10}
    translation {y=-5}
   }
   n22 = "Transform (Modify) Bone - Bone: dyn_vestBack_r_01"(ComponentPose=@n21, Alpha=@n18, Translation="0,5,0", Rotation="0,10,0", Scale="1,1,1") {
    rotationMode=BMM_Additive
    rotationSpace=BCS_BoneSpace
    translationMode=BMM_Additive
    translationSpace=BCS_BoneSpace
    alphaBoolBlend {blendOption=Linear}
    alphaScaleBiasClamp {bInterpResult=true, interpSpeedDecreasing=1}
    boneToModify {boneName=dyn_vestBack_r_01}
    rotation {yaw=10}
    translation {y=5}
   }
   Output Pose(Result=Component To Local(ComponentPose=@n22)) {layerGroup=None}
   n23 = Control Rig(Source=Copy Pose From Mesh() {bUseAttachedParent=true}) {
    alphaBoolBlend {blendOption=Linear}
    controlRigAssetReference {blueprintRigClass=&/F_MED_Lateral_Deform_CtrlRig_twists.F_MED_Lateral_Deform_CtrlRig_twists_C}
   }
   n24 = Control Rig!(
    Source=@n23
    Alpha
    KneeRot_fr_l_114deg:Vector
    KneeTrans_fr_l_114deg:Vector="0.258,-3.591,0.009"
    KneeScale_fr_l_114deg:Vector="1,1,1"
    KneeScale_fr_l_130deg:Vector="1,1,1"
    KneeRot_fr_l_130deg:Vector="0,0,-51.2"
    KneeTrans_fr_l_130deg:Vector="0.258,-3.591,0.009"
   ) -> (Pose) {
    alphaBoolBlend {blendOption=Linear}
    controlRigAssetReference {blueprintRigClass=&/F_MED_Lateral_Deform_CtrlRig_sculpters.F_MED_Lateral_Deform_CtrlRig_sculpters_C}
   }
   Save cached pose 'Anim Cache'(Pose=Component To Local(ComponentPose=Local To Component(LocalPose=@n24))) {cacheName=Anim Cache}
  ]}
 EventGraph {nodes [
   Event BlueprintUpdateAnimation()
   > n1 = Sequence!(execute) -> (then_0, then_1, then_2:Exec, then_3:Exec, then_4:Exec, then_5:Exec, then_6:Exec, then_7:Exec, then_8:Exec)
   > Set thigh L Rot(thigh L Rot:double=GetAlphaFromReferenceTransformDelta(
     SkeletalMeshComponent=Get body Skeleton() -> (body Skeleton:Skeletal Mesh Component Object Reference)
     BoneName=thigh_l
     BaseName=pelvis
     InputRange=(X=-50,Y=-20)
     OutputRange=(X=1,Y=0)
     TwistAxis="1,0,0"
    )) -> (Output_Get:double)
   Set body Skeleton(execute=@n1.then_1, body Skeleton:Skeletal Mesh Component Object Reference=GetOwnerSkeletalMeshComponent()) -> (Output_Get:Skeletal Mesh Component Object Reference)
   Set arm L Rot(
    execute=@n1.then_2
    arm L Rot:double=GetAlphaFromReferenceTransformDelta(
     SkeletalMeshComponent=Get body Skeleton()
     BoneName=upperarm_l
     BaseName=spine_05
     DeltaComponentType=EulerY
     InputRange=(X=0,Y=35)
     TwistAxis="1,0,0"
    )
   ) -> (Output_Get:double)
   Set thigh R Rot(
    execute=@n1.then_3
    thigh R Rot:double=GetAlphaFromReferenceTransformDelta(
     SkeletalMeshComponent=Get body Skeleton()
     BoneName=thigh_r
     BaseName=pelvis
     InputRange=(X=-65,Y=-30)
     OutputRange=(X=1,Y=0)
     TwistAxis="1,0,0"
    )
   ) -> (Output_Get:double)
   n2 = Gravity Override State Function DEPRECATED(
    execute=@n1.then_4
    CharacterPartAnimInstance=Self-Reference()
    GameTimeInSeconds=GetGameTimeInSeconds()
    GravityOverrideParams=Get Gravity Override Params()
   )
   > Set Gravity Override(Gravity Override=@n2)
   n3 = Branch(execute=@n1.then_5, Condition=Get bIsJumping())
   > Set start jump(start jump:double=1) -> (Output_Get:double)
   Set start jump(execute=@n3.else, start jump=0)
   Set head Rot Back(
    execute=@n1.then_6
    head Rot Back:double=GetAlphaFromReferenceTransformDelta(
     SkeletalMeshComponent=Get body Skeleton()
     BoneName=head
     BaseName=spine_05
     InputRange=(X=-50,Y=10)
     OutputRange=(X=1,Y=0)
     TwistAxis="1,0,0"
    )
   ) -> (Output_Get:double)
   Set isFrontEndAlpha(execute=@n1.then_7, isFrontEndAlpha:double="To Float (Boolean)"(InBool=Get bIsInFrontEnd())) -> (Output_Get:double)
   n4 = Get bIsSkyDiving()
   n5 = Get bIsSkydiveDiveMode()
   n6 = Get bIsSkydiveFloating()
   n7 = Get bIsParachuteOpened()
   Set Is in Air(execute=@n1.then_8, Is in Air:double="To Float (Boolean)"(InBool=OR Boolean!(A=@n4, B=@n5, C:bool=@n6, D:bool=@n7) -> (ReturnValue))) -> (Output_Get:double)
   Event OnBeginCrouching()
   > Set Crouching(Crouching:bool=true) -> (Output_Get:bool)
   Event OnEndCrouching()
   > Set Crouching(Crouching=false)
  ]}
]
variables [
 arm L Rot {type=number}
 body Skeleton {type=SkeletalMeshComponent}
 Character Part Anim Instance {type=CustomCharacterPartAnimInstance, default=null}
 Crouching {type=boolean}
 Gravity Override {type=Vector}
 Gravity Override Params {type=GravityOverrideParamsStruct}
 head Rot Back {type=number}
 Is in Air {type=number}
 isFrontEndAlpha {type=number}
 Skeletal Mesh Component {type=SkinnedMeshComponent, default=null}
 start jump {type=number}
 thigh L Rot {type=number}
 thigh R Rot {type=number}
]
