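# Physics asset for the face accessory part of the Lateral medium female player body, defining capsule physics for dynamic hair bones.
# Front and side hair form short chains from the head, with left and right front chains forking at their tips, while top and mesh pieces attach individually.
# Hair joints have limited swing and mostly zero twist.
# Multi-capsule head and upper spine, plus neck and upper arms, approximate the body as collision volumes.
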PhysicsAsset /BRCosmetics/Characters/Player/Female/Medium/Bodies/F_MED_Lateral/Meshes/Parts/F_MED_Lateral_FaceAcc_Physics
bodies {
 dyn_hairFront_c_01 [{shapeType=Capsule, length=1, radius=2, shapeName=dyn_hairFront_c_01_capsule, center=<3.4124 -0.463 0.0001>, rotation=<rot 90 0 0>}]
 dyn_hairFront_c_02 [{shapeType=Capsule, length=0.75, radius=2.75, shapeName=dyn_hairFront_c_02_capsule, center=<1.7101 0 0.0001>, rotation=<rot 90 0 0>}]
 dyn_hairFront_c_03 [{shapeType=Capsule, length=0, radius=3.25, shapeName=dyn_hairFront_c_03_capsule}]
 dyn_hairFront_l_01 [{shapeType=Capsule, length=0.5, radius=2.5, shapeName=dyn_hairFront_l_01_capsule, center=<2.755 -0.0001 1.4183>, rotation=<rot 90 0 0>}]
 dyn_hairFront_l_02 [{shapeType=Capsule, length=1.5, radius=2.75, shapeName=dyn_hairFront_l_02_capsule, center=<3.3113 0 0.5886>, rotation=<rot 90 0 0>}]
 dyn_hairFront_l_03 [{shapeType=Capsule, length=0.5, radius=2.75, shapeName=dyn_hairFront_l_03_capsule, center=<1.1069 0 0>, rotation=<rot 90 0 0>}]
 dyn_hairFront_l_04 [{shapeType=Capsule, length=1.25, radius=1.75, shapeName=dyn_hairFront_l_04_capsule, center=<2.0842 0 0>, rotation=<rot 90 0 0>}]
 dyn_hairFront_l_05 [{shapeType=Capsule, length=4.5, radius=1.25, shapeName=dyn_hairFront_l_05_capsule, center=<2.32 0 0>, rotation=<rot 90 0 0>}]
 dyn_hairFront_r_01 [{shapeType=Capsule, length=5, radius=1.25, shapeName=dyn_hairFront_r_01_capsule, center=<-2.826 0 -1.0816>, rotation=<rot 90 0 0>}]
 dyn_hairFront_r_02 [{shapeType=Capsule, length=0.25, radius=3, shapeName=dyn_hairFront_r_02_capsule, center=<-3.0598 0 0>, rotation=<rot 90 0 0>}]
 dyn_hairFront_r_03 [{shapeType=Capsule, length=0, radius=3.5, shapeName=dyn_hairFront_r_03_capsule, rotation=<rot 90 0 0>}]
 dyn_hairFront_r_04 [{
   shapeType=Capsule
   length=2.75
   radius=1.5
   shapeName=dyn_hairFront_r_04_capsule
   center=<-1.9166 -0.0001 -0.6015>
   rotation=<rot 71.8056 179.9996 179.9996>
  }]
 dyn_hairFront_r_05 [{shapeType=Capsule, length=3.75, radius=1.75, shapeName=dyn_hairFront_r_05_capsule, center=<-3.3849 0.0001 -0.0002>, rotation=<rot 90 0 0>}]
 dyn_hairMeshTop_c_01 [{
   shapeType=Capsule
   length=1.75
   radius=1.25
   shapeName=dyn_hairMeshTop_c_01_capsule
   center=<1.5506 -0.0002 -0.8194>
   rotation=<rot 50.4759 0.0001 0.0001>
  }]
 dyn_hairMeshTop_c_02 [{shapeType=Capsule, length=2.25, radius=1.25, shapeName=dyn_hairMeshTop_c_02_capsule, center=<2.2357 0 0>, rotation=<rot 69.7404 0 0.0001>}]
 dyn_hairMeshTop_c_03 [{shapeType=Capsule, length=2.25, radius=1.5, shapeName=dyn_hairMeshTop_c_03_capsule, center=<1.7769 0 0>, rotation=<rot 79.7847 -0.0001 0>}]
 dyn_hairMesh_l_01 [{shapeType=Capsule, length=4, radius=0.75, shapeName=dyn_hairMesh_l_01_capsule, center=<2.6078 0 0>, rotation=<rot 90 0 0>}]
 dyn_hairMesh_r_01 [{shapeType=Capsule, length=2.25, radius=0.75, shapeName=dyn_hairMesh_r_01_capsule, center=<1.8622 0.4163 0>, rotation=<rot 90 0 -10.0001>}]
 dyn_hairSide_l_01 [{shapeType=Capsule, length=1.25, radius=1.75, shapeName=dyn_hairSide_l_01_capsule, center=<2.0346 -1.4644 0>, rotation=<rot 90 0 0>}]
 dyn_hairSide_l_02 [{shapeType=Capsule, length=0.5, radius=2.5, shapeName=dyn_hairSide_l_02_capsule, center=<4 -1.3798 0>, rotation=<rot 90 0 0>}]
 dyn_hairSide_r_01 [{shapeType=Capsule, length=1, radius=2.5, shapeName=dyn_hairSide_r_01_capsule, center=<-3.1064 1.8395 0.0001>, rotation=<rot 90 0 0>}]
 dyn_hairSide_r_02 [{shapeType=Capsule, length=0.5, radius=2.75, shapeName=dyn_hairSide_r_02_capsule, center=<-3.1775 1.5187 1.5824>, rotation=<rot 90 0 0>}]
 head [
  {shapeType=Capsule, length=8.75, radius=6.75, shapeName=head_capsule, center=<3.7178 0.5734 -0.0002>, rotation=<rot 90 0 0>}
  {shapeType=Capsule, length=1.75, radius=6.75, shapeName=head_capsule, center=<4.4584 2.7967 0>, rotation=<rot 0 0 89.9999>}
  {shapeType=Capsule, length=21, radius=3.5, shapeName=head_capsule, center=<-7.7874 8.8751 3.3204>, rotation=<rot 81.6936 -89.994 -0.0002>}
  {shapeType=Capsule, length=3.5, radius=6.75, shapeName=head_capsule, center=<8.0348 0.0001 0>, rotation=<rot 0 0 80.8894>}
 ]
 neck_01 [{shapeType=Capsule, length=9, radius=5.755, shapeName=neck_01_capsule, center=<-0.0294 2.6654 0.0001>, rotation=<rot 90 -90 -83.5458>}]
 spine_05 [
  {shapeType=Capsule, length=19, radius=9.5, shapeName=spine_05_capsule, center=<0 -1.1259 0.0001>}
  {shapeType=Capsule, length=15, radius=15, shapeName=spine_05_capsule, center=<14.0055 -26.2337 0>, rotation=<rot 0 38.6398 0>}
  {
   shapeType=Capsule
   length=11
   radius=4.5
   shapeName=spine_05_capsule
   center=<12.0381 -10.0287 -15.06>
   rotation=<rot -4.7929 10.8864 -61.6895>
  }
  {
   shapeType=Capsule
   length=8.25
   radius=2.25
   shapeName=spine_05_capsule
   center=<11.9432 -7.306 10.6317>
   rotation=<rot 5.4813 8.4021 67.7477>
  }
  {
   shapeType=Capsule
   length=19
   radius=8
   shapeName=spine_05_capsule
   center=<22.1405 -5.6086 -18.2844>
   rotation=<rot -10.7711 22.4515 -58.9706>
  }
  {
   shapeType=Capsule
   length=9.25
   radius=7.5
   shapeName=spine_05_capsule
   center=<10.7394 -28.814 34.5535>
   rotation=<rot 7.2247 174.6262 126.7738>
  }
  {
   shapeType=Capsule
   length=5.75
   radius=2
   shapeName=spine_05_capsule
   center=<10.4972 1.8993 -6.0736>
   rotation=<rot 11.1983 25.8772 56.9675>
  }
 ]
 upperarm_l [{
   shapeType=Capsule
   length=14.5
   radius=6.005
   shapeName=upperarm_l_capsule
   center=<13.7753 -0.0004 -0.2254>
   rotation=<rot 0 -89.9998 -91.1475>
  }]
 upperarm_r [{
   shapeType=Capsule
   length=13
   radius=8.255
   shapeName=upperarm_r_capsule
   center=<-10.9407 0.2071 0.7581>
   rotation=<rot 0 -89.9996 -91.1474>
  }]
}
constraints [
 {bone1Name=dyn_hairFront_c_02, bone2Name=dyn_hairFront_c_01, swing1LimitDegrees=30, swing2LimitDegrees=20, twistLimitDegrees=0}
 {bone1Name=dyn_hairFront_c_03, bone2Name=dyn_hairFront_c_02, swing1LimitDegrees=30, swing2LimitDegrees=20, twistLimitDegrees=0}
 {bone1Name=dyn_hairSide_l_02, bone2Name=dyn_hairSide_l_01, swing1LimitDegrees=20, swing2LimitDegrees=20, twistLimitDegrees=0}
 {bone1Name=dyn_hairFront_l_02, bone2Name=dyn_hairFront_l_01, swing1LimitDegrees=20, swing2LimitDegrees=20, twistLimitDegrees=0}
 {bone1Name=dyn_hairFront_l_03, bone2Name=dyn_hairFront_l_02, swing1LimitDegrees=30, swing2LimitDegrees=30, twistLimitDegrees=0}
 {bone1Name=dyn_hairFront_l_04, bone2Name=dyn_hairFront_l_03, swing1LimitDegrees=20, swing2LimitDegrees=20, twistLimitDegrees=0}
 {bone1Name=dyn_hairFront_l_05, bone2Name=dyn_hairFront_l_03, swing1LimitDegrees=20, swing2LimitDegrees=20, twistLimitDegrees=0}
 {bone1Name=dyn_hairSide_r_02, bone2Name=dyn_hairSide_r_01, swing1LimitDegrees=30, swing2LimitDegrees=30, twistLimitDegrees=0}
 {bone1Name=dyn_hairFront_r_02, bone2Name=dyn_hairFront_r_01, swing1LimitDegrees=40, swing2LimitDegrees=40, twistLimitDegrees=20}
 {bone1Name=dyn_hairFront_r_03, bone2Name=dyn_hairFront_r_02, swing1LimitDegrees=40, swing2LimitDegrees=40, twistLimitDegrees=20}
 {bone1Name=dyn_hairFront_r_04, bone2Name=dyn_hairFront_r_03, swing1LimitDegrees=20, swing2LimitDegrees=20, twistLimitDegrees=0}
 {bone1Name=dyn_hairFront_r_05, bone2Name=dyn_hairFront_r_03, swing1LimitDegrees=20, swing2LimitDegrees=20, twistLimitDegrees=0}
 {bone1Name=dyn_hairFront_c_01, bone2Name=head, swing1LimitDegrees=30, swing2LimitDegrees=20, twistLimitDegrees=0}
 {bone1Name=dyn_hairSide_l_01, bone2Name=head, swing1LimitDegrees=10, swing2LimitDegrees=10, twistLimitDegrees=0}
 {bone1Name=dyn_hairFront_l_01, bone2Name=head, swing1LimitDegrees=25, swing2LimitDegrees=15, twistLimitDegrees=0}
 {bone1Name=dyn_hairSide_r_01, bone2Name=head, swing1LimitDegrees=20, swing2LimitDegrees=20, twistLimitDegrees=0}
 {bone1Name=dyn_hairFront_r_01, bone2Name=head, swing1LimitDegrees=20, swing2LimitDegrees=40, twistLimitDegrees=0}
 {bone1Name=dyn_hairMeshTop_c_01, bone2Name=head, swing1LimitDegrees=20, swing2LimitDegrees=20, twistLimitDegrees=0}
 {bone1Name=dyn_hairMeshTop_c_02, bone2Name=head, swing1LimitDegrees=20, swing2LimitDegrees=20, twistLimitDegrees=0}
 {bone1Name=dyn_hairMeshTop_c_03, bone2Name=head, swing1LimitDegrees=20, swing2LimitDegrees=20, twistLimitDegrees=0}
 {bone1Name=dyn_hairMesh_l_01, bone2Name=head, swing1LimitDegrees=20, swing2LimitDegrees=20, twistLimitDegrees=0}
 {bone1Name=dyn_hairMesh_r_01, bone2Name=head, swing1LimitDegrees=20, swing2LimitDegrees=20, twistLimitDegrees=0}
 {bone1Name=head, bone2Name=neck_01, swing1LimitDegrees=20, swing2LimitDegrees=20, twistLimitDegrees=0}
 {
  bone1Name=neck_01
  bone2Name=spine_05
  swing1LimitDegrees=45
  swing1Motion=Free
  swing2LimitDegrees=45
  swing2Motion=Free
  twistLimitDegrees=45
  twistMotion=Free
 }
 {
  bone1Name=upperarm_l
  bone2Name=spine_05
  swing1LimitDegrees=45
  swing1Motion=Free
  swing2LimitDegrees=45
  swing2Motion=Free
  twistLimitDegrees=45
  twistMotion=Free
 }
 {
  bone1Name=upperarm_r
  bone2Name=spine_05
  swing1LimitDegrees=45
  swing1Motion=Free
  swing2LimitDegrees=45
  swing2Motion=Free
  twistLimitDegrees=45
  twistMotion=Free
 }
]
