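# Post-physics hair correction AnimBP for the F_MED_Lateral female face accessory mesh.
# It copies the attached parent's pose and curves, layers the body mesh pose from both upper arms, then runs RigidBody simulation relative to spine_05.
# Modify Bone nodes then reshape right-side hair, weighted by head rotation during pistol aiming or by a front-end flag.
# The event graph derives these alphas from head-to-spine twist and player anim instance state.
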AnimBlueprint /BRCosmetics/Characters/Player/Female/Medium/Bodies/F_MED_Lateral/Meshes/Parts/F_MED_Lateral_FaceAcc_AnimBP
target_skeleton=&/F_MED_Lateral_FaceAcc_Skeleton
graphs [
 AnimGraph {nodes [
   n1 = Copy Pose From Mesh(SourceMeshComponent=Get Body_Mesh())
   n2 = Copy Pose From Mesh() {bCopyCurves=true, bUseAttachedParent=true}
   n3 = RigidBody(ComponentPose=Local To Component(LocalPose=Layered blend per bone(BasePose=@n2, BlendPoses_0=@n1) {layerSetup [{branchFilters [{boneName=upperarm_l}, {boneName=upperarm_r}]}]})) {
    bOverrideWorldGravity=true
    overridePhysicsAsset=&/F_MED_Lateral_FaceAcc_Physics
    simulationSpace=BaseBoneSpace
    alphaBoolBlend {blendOption=Linear}
    baseBoneRef {boneName=spine_05}
    componentAppliedLinearAccClamp=<3000 3000 3000>
    componentLinearAccScale=<0.25 0.25 0.45>
    componentLinearVelScale=<0.25 0.25 0.45>
    overrideWorldGravity {z=-980}
   }
   n4 = Get head Rot Pistol On() -> (head Rot Pistol On:double)
   n5 = "Transform (Modify) Bone - Bone: dyn_hairFront_r_01"(ComponentPose=@n3, Alpha=@n4, Translation="0,0,0", Rotation="0,0,0", Scale="0.9,1,1") {
    rotationMode=BMM_Additive
    rotationSpace=BCS_BoneSpace
    scaleMode=BMM_Replace
    scaleSpace=BCS_BoneSpace
    translationMode=BMM_Additive
    translationSpace=BCS_BoneSpace
    alphaBoolBlend {blendOption=Linear}
    boneToModify {boneName=dyn_hairFront_r_01}
    scale {x=0.9}
   }
   n6 = "Transform (Modify) Bone - Bone: dyn_hairSide_r_01"(ComponentPose=@n5, Alpha=@n4, Translation="3,0,0", Rotation="0,0,0", Scale="0.1,1,1") {
    rotationMode=BMM_Additive
    rotationSpace=BCS_BoneSpace
    scaleMode=BMM_Replace
    scaleSpace=BCS_BoneSpace
    translationMode=BMM_Additive
    translationSpace=BCS_BoneSpace
    alphaBoolBlend {blendOption=Linear}
    boneToModify {boneName=dyn_hairSide_r_01}
    scale {x=0.1}
    translation {x=3}
   }
   n7 = "Transform (Modify) Bone - Bone: dyn_hairSide_r_02"(ComponentPose=@n6, Alpha=@n4, Translation="3,0,0", Rotation="0,0,0", Scale="0.1,1,0.5") {
    rotationMode=BMM_Additive
    rotationSpace=BCS_BoneSpace
    scaleMode=BMM_Replace
    scaleSpace=BCS_BoneSpace
    translationMode=BMM_Additive
    translationSpace=BCS_BoneSpace
    alphaBoolBlend {blendOption=Linear}
    boneToModify {boneName=dyn_hairSide_r_02}
    scale {x=0.1, z=0.5}
    translation {x=3}
   }
   n8 = Get Is In Front End() -> (Is In Front End:double)
   n9 = "Transform (Modify) Bone - Bone: dyn_hairFront_r_02"(ComponentPose=@n7, Alpha=@n8, Translation="-1,4,4", Rotation="15,0,0", Scale="1,1.3,1.1") {
    rotationMode=BMM_Additive
    rotationSpace=BCS_BoneSpace
    scaleMode=BMM_Replace
    scaleSpace=BCS_BoneSpace
    translationMode=BMM_Additive
    translationSpace=BCS_BoneSpace
    alphaBoolBlend {blendOption=Linear}
    boneToModify {boneName=dyn_hairFront_r_02}
    rotation {pitch=15}
    scale {y=1.3, z=1.1}
    translation=<-1 4 4>
   }
   n10 = "Transform (Modify) Bone - Bone: dyn_hairFront_r_05"(ComponentPose=@n9, Alpha=@n8, Translation="0,0,-3", Rotation="0,0,0", Scale="1,1,1.5") {
    rotationMode=BMM_Additive
    rotationSpace=BCS_BoneSpace
    scaleMode=BMM_Replace
    translationMode=BMM_Additive
    translationSpace=BCS_BoneSpace
    alphaBoolBlend {blendOption=Linear}
    boneToModify {boneName=dyn_hairFront_r_05}
    scale {z=1.5}
    translation {z=-3}
   }
   n11 = "Transform (Modify) Bone - Bone: dyn_hairSide_r_02"(ComponentPose=@n10, Alpha=@n8, Translation="-3,0,0", Rotation="0,0,0", Scale="1,1.2,1") {
    rotationMode=BMM_Additive
    rotationSpace=BCS_BoneSpace
    translationMode=BMM_Additive
    translationSpace=BCS_BoneSpace
    alphaBoolBlend {blendOption=Linear}
    boneToModify {boneName=dyn_hairSide_r_02}
    scale {y=1.2}
    translation {x=-3}
   }
   Output Pose(Result=Component To Local(ComponentPose=@n11)) {layerGroup=None}
  ]}
 EventGraph {nodes [
   Event BlueprintUpdateAnimation()
   > n1 = Sequence!(execute) -> (then_0, then_1, then_2:Exec, then_3:Exec)
   > n2 = Set head Rot Back(head Rot Back:double=GetAlphaFromReferenceTransformDelta(
     SkeletalMeshComponent=Get body Mesh() -> (body Mesh:Skeletal Mesh Component Object Reference)
     BoneName=head
     BaseName=spine_05
     InputRange=(X=-40,Y=0)
     OutputRange=(X=1,Y=0)
     TwistAxis="1,0,0"
    )) -> (Output_Get:double)
   n3 = Set using Pistol(execute=@n7.then, using Pistol:bool=true) -> (Output_Get:bool)
   n4 = Branch(execute=@n2.then, Condition=@n3)
   > Set head Rot Pistol On(head Rot Pistol On:double=@n2) -> (Output_Get:double)
   Set head Rot Pistol On(execute=@n4.else, head Rot Pistol On=0)
   Set body Mesh(execute=@n1.then_1, body Mesh:Skeletal Mesh Component Object Reference=GetOwnerSkeletalMeshComponent()) -> (Output_Get:Skeletal Mesh Component Object Reference)
   n5 = Cast To FortPlayerAnimInstance(execute=@n1.then_2, Object=GetAnimInstance(self=GetOwnerSkeletalMeshComponent()))
   > Set ar_firingPosition(ar_firingPosition:double=0) -> (Output_Get:double)
   > n6 = Switch on EFortWeaponCoreAnimation!(execute, Selection:EFortWeaponCoreAnimation Enum=Get WeaponCoreAnim(self=@n5)) -> (
    Melee:Exec
    Pistol:Exec
    Shotgun:Exec
    PaperBlueprint:Exec
    Rifle:Exec
    MeleeOneHand:Exec
    MachinePistol:Exec
    RocketLauncher:Exec
    GrenadeLauncher:Exec
    GoingCommando:Exec
    AssaultRifle:Exec
    TacticalShotgun:Exec
    SniperRifle:Exec
    TrapPlacement:Exec
    ShoulderLauncher:Exec
    AbilityDecoTool:Exec
    Crossbow:Exec
    C4:Exec
    RemoteControl:Exec
    DualWield:Exec
    AR_BullPup:Exec
    AR_ForwardGrip:Exec
    MedPackPaddles:Exec
    SMG_P90:Exec
    AR_DrumGun:Exec
    Consumable_Small:Exec
    Consumable_Large:Exec
    Balloon:Exec
    MountedTurret:Exec
    CreativeTool:Exec
    ExplosiveBow:Exec
    AshtonIndigo:Exec
    AshtonChicago:Exec
    MeleeDualWield:Exec
    Unarmed:Exec
   )
   n7 = Branch(execute=@n6.Pistol, Condition=Get bIsTargeting())
   Set using Pistol(execute=@n7.else, using Pistol=false)
   Set Is In Front End(execute=@n1.then_3, Is In Front End:double="To Float (Boolean)"(InBool=Get bIsInFrontEnd())) -> (Output_Get:double)
   Event BlueprintInitializeAnimation()
   > Set Body_Mesh(Body_Mesh=GetOwnerSkeletalMeshComponent())
  ]}
]
variables [
 Alt_weapon {type=boolean}
 ar_firingPosition {type=number}
 body Mesh {type=SkeletalMeshComponent}
 Body_Mesh {type=SkeletalMeshComponent}
 head Rot Back {type=number}
 head Rot Pistol On {type=number}
 Is In Front End {type=number}
 Skeletal Mesh Component {type=SkinnedMeshComponent, default=null}
 using Pistol {type=boolean}
]
